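# Post-pose secondary-motion AnimBlueprint for the F_MED_CommandoSpy female player body, animating belt pieces, zip chain and collar.
# The parent mesh's pose feeds a Control Rig, a two-spine zip-chain constraint, then sprung Anim Dynamics bones with a thigh collision sphere.
# Additive offsets follow, shifting the left side belt on left/right movement and pitching the right collar.
# EventGraph derives movement, jump, sprint, emote, front-end and gravity variables, several unused by the AnimGraph.
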AnimBlueprint /BRCosmetics/Characters/Player/Female/Medium/Bodies/F_MED_CommandoSpy/Meshes/F_MED_CommandoSpy_AnimBP
target_skeleton=&/F_MED_CommandoSpy_Skeleton
graphs [
 AnimGraph {nodes [
   n1 = Control Rig!(
    Source=Copy Pose From Mesh() {bUseAttachedParent=true}
    Alpha
    Not In FE:bool=Get Not FE() -> (Not FE:bool)
    targeting:bool=Get bIsTargeting()
   ) -> (Pose) {
    alphaBoolBlend {blendOption=Linear}
    controlRigAssetReference {blueprintRigClass=&/F_MED_CommandoSpy_CtrlRig.F_MED_CommandoSpy_CtrlRig_C}
   }
   n2 = "Constraint - dyn_zip_chain (PP)"!(ComponentPose=Local To Component(LocalPose=@n1), Alpha=1, ConstraintWeights_0=0.9694, ConstraintWeights_1:float=0.0306) -> (Pose) {
    alphaBoolBlend {blendOption=Linear}
    boneToModify {boneName=dyn_zip_chain}
    constraintSetup [{targetBone {boneName=spine_05}}, {targetBone {boneName=spine_04}}]
    constraintWeights [1,1]
   }
   n3 = Anim Dynamics - Bone: dyn_Belt_belt_side_l!(
    SimulationSpace:AnimPhysSimSpaceType Enum
    ComponentPose=@n2
    bAlphaBoolEnabled:bool=True
    GravityOverride:Vector="0,0,-980"
    ExternalForce="0,0,0"
   ) -> (Pose) {
    alphaInputType=Bool
    angularDampingOverride=1
    angularSpringConstant=10000
    bAngularSpring=true
    bOverrideAngularDamping=true
    bUseGravityOverride=true
    bUseSphericalLimits=true
    gravityScale=0.35
    alphaBoolBlend {blendOption=Linear}
    boundBone {boneName=dyn_Belt_belt_side_l}
    gravityOverride {z=-980}
    physicsBodyDefinitions [{
      boundBone {boneName=dyn_Belt_belt_side_l}
      boxExtents=<20 30 30>
      constraintSetup {angularLimitsMax=<0 0 0>, angularLimitsMin=<0 0 0>, angularTarget=<1 0 0>}
      localJointOffset=<1 0 0>
     }]
    sphericalLimits [{limitRadius=10, drivingBone {boneName=thigh_l}, sphereLocalOffset=<-4.5615 0 1.8715>}]
   }
   n4 = Anim Dynamics - Bone: dyn_Belt_belt_back_l!(
    SimulationSpace:AnimPhysSimSpaceType Enum
    ComponentPose=@n3
    bAlphaBoolEnabled:bool=True
    GravityOverride:Vector="0,0,-980"
    ExternalForce="0,0,0"
   ) -> (Pose) {
    alphaInputType=Bool
    angularDampingOverride=1
    angularSpringConstant=10000
    bAngularSpring=true
    bOverrideAngularDamping=true
    bUseGravityOverride=true
    gravityScale=0.35
    alphaBoolBlend {blendOption=Linear}
    boundBone {boneName=dyn_Belt_belt_back_l}
    gravityOverride {z=-980}
    physicsBodyDefinitions [{
      boundBone {boneName=dyn_Belt_belt_back_l}
      boxExtents=<30 40 40>
      constraintSetup {angularLimitsMax=<0 0 0>, angularLimitsMin=<0 0 0>, angularTarget=<1 0 0>}
      localJointOffset=<1 0 0>
     }]
   }
   n5 = Anim Dynamics - Bone: dyn_Belt_belt_frnt_center!(
    SimulationSpace:AnimPhysSimSpaceType Enum
    ComponentPose=@n4
    bAlphaBoolEnabled:bool=True
    GravityOverride:Vector="0,0,-980"
    ExternalForce="0,0,0"
   ) -> (Pose) {
    alphaInputType=Bool
    angularDampingOverride=1
    angularSpringConstant=10000
    bAngularSpring=true
    bOverrideAngularDamping=true
    bUseGravityOverride=true
    gravityScale=0.35
    alphaBoolBlend {blendOption=Linear}
    boundBone {boneName=dyn_Belt_belt_frnt_center}
    gravityOverride {z=-980}
    physicsBodyDefinitions [{
      boundBone {boneName=dyn_Belt_belt_frnt_center}
      boxExtents=<20 30 30>
      constraintSetup {angularLimitsMax=<0 0 0>, angularLimitsMin=<0 0 0>, angularTarget=<1 0 0>}
      localJointOffset=<4 0 0>
     }]
   }
   n6 = Anim Dynamics - Bone: dyn_Belt_belt_back_r!(
    SimulationSpace:AnimPhysSimSpaceType Enum
    ComponentPose=@n5
    bAlphaBoolEnabled:bool=True
    GravityOverride:Vector="0,0,-980"
    ExternalForce="0,0,0"
   ) -> (Pose) {
    alphaInputType=Bool
    angularDampingOverride=1
    angularSpringConstant=10000
    bAngularSpring=true
    bOverrideAngularDamping=true
    bUseGravityOverride=true
    gravityScale=0.35
    alphaBoolBlend {blendOption=Linear}
    boundBone {boneName=dyn_Belt_belt_back_r}
    gravityOverride {z=-980}
    physicsBodyDefinitions [{
      boundBone {boneName=dyn_Belt_belt_back_r}
      boxExtents=<30 40 40>
      constraintSetup {angularLimitsMax=<0 0 0>, angularLimitsMin=<0 0 0>, angularTarget=<1 0 0>}
      localJointOffset=<1 0 0>
     }]
   }
   n7 = Anim Dynamics - Bone: dyn_Belt_belt_side_r!(
    SimulationSpace:AnimPhysSimSpaceType Enum
    ComponentPose=@n6
    bAlphaBoolEnabled:bool=True
    GravityOverride:Vector="0,0,-980"
    ExternalForce="0,0,0"
   ) -> (Pose) {
    alphaInputType=Bool
    angularDampingOverride=1
    angularSpringConstant=10000
    bAngularSpring=true
    bOverrideAngularDamping=true
    bUseGravityOverride=true
    gravityScale=0.35
    alphaBoolBlend {blendOption=Linear}
    boundBone {boneName=dyn_Belt_belt_side_r}
    gravityOverride {z=-980}
    physicsBodyDefinitions [{
      boundBone {boneName=dyn_Belt_belt_side_r}
      boxExtents=<30 40 40>
      constraintSetup {angularLimitsMax=<0 0 0>, angularLimitsMin=<0 0 0>, angularTarget=<1 0 0>}
      localJointOffset=<1 0 0>
     }]
   }
   n8 = Anim Dynamics - Bone: dyn_zip_chain!(SimulationSpace:AnimPhysSimSpaceType Enum, ComponentPose=@n7, Alpha=1, GravityOverride:Vector="0,0,-980", ExternalForce="0,0,0") -> (Pose) {
    angularDampingOverride=1
    angularSpringConstant=15000
    bAngularSpring=true
    bOverrideAngularDamping=true
    bUseGravityOverride=true
    gravityScale=0.35
    simulationSpace=RootRelative
    alphaBoolBlend {blendOption=Linear}
    boundBone {boneName=dyn_zip_chain}
    componentLinearAccScale=<1 1 1>
    componentLinearVelScale=<1 1 1>
    gravityOverride {z=-980}
    physicsBodyDefinitions [{
      boundBone {boneName=dyn_zip_chain}
      boxExtents=<20 30 30>
      constraintSetup {angularLimitsMax=<0 0 15>, angularLimitsMin=<0 -30 -15>, angularTarget=<1 0 0>}
      localJointOffset=<2 0 0>
     }]
   }
   n9 = Anim Dynamics - Bone: dyn_collar_front_l!(ComponentPose=@n8, Alpha=1, GravityOverride:Vector="0,0,-980", ExternalForce="0,0,0") -> (Pose) {
    angularDampingOverride=1
    angularSpringConstant=8000
    bAngularSpring=true
    bOverrideAngularDamping=true
    bUseGravityOverride=true
    gravityScale=0.35
    simulationSpace=RootRelative
    alphaBoolBlend {blendOption=Linear}
    boundBone {boneName=dyn_collar_front_l}
    componentLinearAccScale=<1 1 1>
    componentLinearVelScale=<1 1 1>
    gravityOverride {z=-980}
    physicsBodyDefinitions [{
      boundBone {boneName=dyn_collar_front_l}
      boxExtents=<10 20 20>
      constraintSetup {angularLimitsMax=<0 0 5>, angularLimitsMin=<0 -2 -5>, angularTarget=<1 0 0>}
      localJointOffset=<4 0 0>
     }]
   }
   n10 = Anim Dynamics - Bone: dyn_collar_front_r!(ComponentPose=@n9, Alpha=1, GravityOverride:Vector="0,0,-980", ExternalForce="0,0,0") -> (Pose) {
    angularDampingOverride=1
    angularSpringConstant=8000
    bAngularSpring=true
    bOverrideAngularDamping=true
    bUseGravityOverride=true
    gravityScale=0.35
    simulationSpace=RootRelative
    alphaBoolBlend {blendOption=Linear}
    boundBone {boneName=dyn_collar_front_r}
    componentLinearAccScale=<1 1 1>
    componentLinearVelScale=<1 1 1>
    gravityOverride {z=-980}
    physicsBodyDefinitions [{
      boundBone {boneName=dyn_collar_front_r}
      boxExtents=<10 20 20>
      constraintSetup {angularLimitsMax=<0 0 5>, angularLimitsMin=<0 -2 -5>, angularTarget=<1 0 0>}
      localJointOffset=<4 0 0>
     }]
   }
   n11 = "Transform (Modify) Bone - Bone: dyn_Belt_belt_side_l"(ComponentPose=@n10, Alpha=Get Left() -> (Left:double), Translation="-2,0,0", Rotation="0,0,0", Scale="1,1,1") {
    rotationMode=BMM_Additive
    rotationSpace=BCS_BoneSpace
    translationMode=BMM_Additive
    translationSpace=BCS_BoneSpace
    alphaBoolBlend {blendOption=Linear}
    alphaScaleBiasClamp {bInterpResult=true, interpSpeedDecreasing=7, interpSpeedIncreasing=7}
    boneToModify {boneName=dyn_Belt_belt_side_l}
    translation {x=-2}
   }
   n12 = "Transform (Modify) Bone - Bone: dyn_Belt_belt_side_l"(ComponentPose=@n11, Alpha=Get Right() -> (Right:double), Translation="-2,0,0", Rotation="0,0,0", Scale="1,1,1") {
    rotationMode=BMM_Additive
    rotationSpace=BCS_BoneSpace
    translationMode=BMM_Additive
    translationSpace=BCS_BoneSpace
    alphaBoolBlend {blendOption=Linear}
    alphaScaleBiasClamp {bInterpResult=true, interpSpeedDecreasing=7, interpSpeedIncreasing=7}
    boneToModify {boneName=dyn_Belt_belt_side_l}
    translation {x=-2}
   }
   n13 = "Transform (Modify) Bone - Bone: dyn_collar_front_r"!(ComponentPose=@n12, bAlphaBoolEnabled:bool=True, Translation="0,0,0", Rotation="-40,0,0", Scale="1,1,1") -> (Pose) {
    alphaInputType=Bool
    rotationMode=BMM_Additive
    rotationSpace=BCS_BoneSpace
    translationMode=BMM_Additive
    translationSpace=BCS_BoneSpace
    alphaBoolBlend {blendOption=Linear}
    alphaScaleBiasClamp {bInterpResult=true, interpSpeedDecreasing=7, interpSpeedIncreasing=7}
    boneToModify {boneName=dyn_collar_front_r}
    rotation {pitch=-40}
   }
   Output Pose(Result=Component To Local(ComponentPose=@n13)) {layerGroup=None}
  ]}
 EventGraph {nodes [
   n1 = Event BlueprintUpdateAnimation()
   > Set Delta Time X(Delta Time X:double=@n1.DeltaTimeX) -> (Output_Get:double)
   > n2 = Sequence!(execute) -> (then_0, then_1, then_2:Exec, then_3:Exec, then_4:Exec, then_5:Exec)
   n3 = GetOwnerSkeletalMeshComponent()
   n4 = BreakRotator(InRot=BreakTransform(InTransform=GetDeltaTransformFromRefPose(self=@n3, BoneName=thigh_r, BaseName=pelvis)).Rotation)
   n5 = Get bIsFalling()
   n6 = Get bIsJumping()
   Set jump(
    execute=@n2.then_3
    jump:double=float * float(A=MapRangeClamped(Value=@n4.Yaw, InRangeA=0, InRangeB=-45), B:bool=OR Boolean(A=@n6, B=@n5)) {is=title K2Node_PromotableOperator}
   ) -> (Output_Get:double)
   n7 = Get bIsPlayingEmote()
   Set Not Emote(execute=@n2.then_4, Not Emote:bool=OR Boolean(A=NOT Boolean(A=@n7), B=false)) -> (Output_Get:bool)
   n8 = Get bIsInFrontEnd()
   > Set Not FE(Not FE:bool=OR Boolean(A=NOT Boolean(A=@n8), B=false)) -> (Output_Get:bool)
   n9 = "Gravity Override State Function (Parameter Asset) (Nativized)"(
    execute=@n2.then_0
    CharacterPartAnimInstance=Self-Reference()
    GameTimeInSeconds=GetGameTimeInSeconds()
    Parameters=&/F_MED_CommandoSpy_Parameters
   )
   > Set Gravity Override(Gravity Override=@n9)
   n10 = Get bIsCrouching()
   Set Sprinting not crouching(execute=@n2.then_2, Sprinting not crouching:bool=AND Boolean(A=Get bIsSprinting(), B=NOT Boolean(A=@n10))) -> (Output_Get:bool)
   n11 = PlayerMovement(execute=@n2.then_1, AnimInstance=Self-Reference(), XDotProductCoefficient=100, YDotProductCoefficient=100, ZDotProductCoefficient=100)
   > n12 = Branch(Condition=Get bIsCrouching())
   n13 = BreakVector(InVec=MakeVector(X=@n11, Y=@n11.YMovement))
   Set Left(execute=@n12.else, Left:double="To Float (Boolean)"(InBool="float < float"(A=@n13, B=-50))) -> (Output_Get:double)
   > Set Right(Right:double="To Float (Boolean)"(InBool="float > float"(A=@n13, B=50))) -> (Output_Get:double)
   > Set Forward(Forward:double="To Float (Boolean)"(InBool="float > float"(A=@n13.Y, B=40))) -> (Output_Get:double)
   > Set Backward(Backward:double="To Float (Boolean)"(InBool="float < float"(A=@n13.Y, B=-40))) -> (Output_Get:double)
  ]}
]
variables [
 Backward {type=number}
 Delta Time X {type=number}
 Forward {type=number}
 Gravity Override {type=Vector}
 jump {type=number}
 Left {type=number}
 Not Emote {type=boolean}
 Not FE {type=boolean}
 Right {type=number}
 Skel Mesh {type=SkeletalMeshComponent}
 Sprinting not crouching {type=boolean}
]
