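# Physics asset for the CommandoSpy female medium face accessory's front hair, defining collision and joint limits for its left and right strands.
# Each side is a chain of four capsule bodies linked by angular-limited constraints to a small MSTR root, stiffest at the root and looser further down.
# Unconstrained capsules on head, neck, upper spine, clavicles and upper arms approximate the body around the hair.
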PhysicsAsset /BRCosmetics/Characters/Player/Female/Medium/Bodies/F_MED_CommandoSpy/Meshes/Parts/F_MED_CommandoSpy_FaceAcc_FrontHairPhysics
bodies {
 clavicle_l [{shapeType=Capsule, length=8.5, radius=6, shapeName=clavicle_l_capsule, center=<3.9192 -0.564 -0.7301>, rotation=<rot 80 0 -12.2871>}]
 clavicle_r [{shapeType=Capsule, length=8, radius=6, shapeName=clavicle_r_capsule, center=<-3.8697 0.8469 0.5132>, rotation=<rot 80 0 0>}]
 dyn_hair_l_01 [{shapeType=Capsule, length=3, radius=2.5, shapeName=dyn_hair_l_01_capsule, center=<3.1 -0.75 0>, rotation=<rot 0 0 -60>}]
 dyn_hair_l_02 [{shapeType=Capsule, length=3, radius=2.5, shapeName=dyn_hair_l_02_capsule, center=<4 -0.4 0.18>, rotation=<rot 0 0 -58>}]
 dyn_hair_l_03 [{shapeType=Capsule, length=10.159, radius=1.6889, shapeName=dyn_hair_l_03_capsule, center=<3 -0.5 1>, rotation=<rot -5.6169 20.6443 -75.4374>}]
 dyn_hair_l_04 [{shapeType=Capsule, length=9.9995, radius=2, shapeName=dyn_hair_l_04_capsule, center=<0 0 1.4>, rotation=<rot 90 -12.5 0>}]
 dyn_hair_l_MSTR [{shapeType=Capsule, length=1, radius=0.5, shapeName=dyn_hair_l_MSTR_capsule}]
 dyn_hair_r_01 [{shapeType=Capsule, length=2, radius=2.4, shapeName=dyn_hair_r_01_capsule, center=<3 0.24 0.7>, rotation=<rot 0 0 88.1509>}]
 dyn_hair_r_02 [{shapeType=Capsule, length=3, radius=2.5, shapeName=dyn_hair_r_02_capsule, center=<3.0918 0.8 0.9>, rotation=<rot 0 -0.0001 71.0018>}]
 dyn_hair_r_03 [{
   shapeType=Capsule
   length=4.0057
   radius=2.4943
   shapeName=dyn_hair_r_03_capsule
   center=<3.4136 0.9608 0.6872>
   rotation=<rot 77.6653 173.8926 -95.7614>
  }]
 dyn_hair_r_04 [{
   shapeType=Capsule
   length=4.3038
   radius=2.1179
   shapeName=dyn_hair_r_04_capsule
   center=<2.368 0.2538 0.9658>
   rotation=<rot 82.7151 173.0678 -111.7899>
  }]
 dyn_hair_r_MSTR [{shapeType=Capsule, length=1, radius=0.5, shapeName=dyn_hair_r_MSTR_capsule}]
 head [{shapeType=Capsule, length=2.5, radius=5.5, shapeName=head_capsule, center=<-0.174 2.8862 0>, rotation=<rot 90 172.875 -171.4604>}]
 neck_01 [
  {shapeType=Capsule, length=5.8265, radius=4.6735, shapeName=neck_01_capsule, center=<1.5 -1.0465 0>, rotation=<rot 90 0 0>}
  {
   shapeType=Capsule
   length=5.4734
   radius=4.6775
   shapeName=neck_01_capsule
   center=<1.6704 4.0531 1.209>
   rotation=<rot 78.8715 107.4764 -5.3303>
  }
 ]
 spine_05 [
  {shapeType=Capsule, length=6, radius=26, shapeName=spine_05_capsule, center=<-18.2331 -6.2771 0>, rotation=<rot 90 180 -140.3635>}
  {
   shapeType=Capsule
   length=11.8436
   radius=1.4594
   shapeName=spine_05_capsule
   center=<9.357 6.5504 -3.2024>
   rotation=<rot -1.6601 8.6785 -84.0916>
  }
 ]
 upperarm_l [{shapeType=Capsule, length=14.5, radius=4.5, shapeName=upperarm_l_capsule, center=<8.5173 -0.1195 0.1462>, rotation=<rot 82.8138 0 4.4014>}]
 upperarm_r [{
   shapeType=Capsule
   length=23
   radius=3
   shapeName=upperarm_r_capsule
   center=<-11.4913 -0.4316 -1.1384>
   rotation=<rot 86.3895 -0.5037 -8.2584>
  }]
}
constraints [
 {bone1Name=dyn_hair_r_04, bone2Name=dyn_hair_r_03, swing1LimitDegrees=40, swing2LimitDegrees=30, twistLimitDegrees=20}
 {bone1Name=dyn_hair_r_03, bone2Name=dyn_hair_r_02, swing1LimitDegrees=40, swing2LimitDegrees=30, twistLimitDegrees=10}
 {bone1Name=dyn_hair_r_02, bone2Name=dyn_hair_r_01, swing1LimitDegrees=30, swing2LimitDegrees=30, twistLimitDegrees=10}
 {bone1Name=dyn_hair_r_01, bone2Name=dyn_hair_r_MSTR, swing1LimitDegrees=10, swing2LimitDegrees=1, twistLimitDegrees=2}
 {bone1Name=dyn_hair_l_04, bone2Name=dyn_hair_l_03, swing1LimitDegrees=30, swing2LimitDegrees=30, twistLimitDegrees=20}
 {bone1Name=dyn_hair_l_03, bone2Name=dyn_hair_l_02, swing1LimitDegrees=40, swing2LimitDegrees=30, twistLimitDegrees=30}
 {bone1Name=dyn_hair_l_02, bone2Name=dyn_hair_l_01, swing1LimitDegrees=20, swing2LimitDegrees=40, twistLimitDegrees=10}
 {bone1Name=dyn_hair_l_01, bone2Name=dyn_hair_l_MSTR, swing1LimitDegrees=10, swing2LimitDegrees=1, twistLimitDegrees=2}
]
