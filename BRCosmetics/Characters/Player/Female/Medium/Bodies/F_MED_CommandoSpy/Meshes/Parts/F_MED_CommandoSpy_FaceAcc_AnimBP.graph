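# Post-process style Anim Blueprint for the female CommandoSpy FaceAcc part, driving its front and back hair over the pose copied from the attached parent.
# A pre Control Rig feeds cached RigidBody passes, switched by skydive and dive states.
# Bone constraints (some backbling-gated), AnimDynamics and corrective offsets follow, ending in a post Control Rig.
# EventGraph derives gravity, head twist, weapon aim, backbling and emote variables each update.
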AnimBlueprint /BRCosmetics/Characters/Player/Female/Medium/Bodies/F_MED_CommandoSpy/Meshes/Parts/F_MED_CommandoSpy_FaceAcc_AnimBP
target_skeleton=&/F_MED_CommandoSpy_FaceAcc_Skeleton
graphs [
 AnimGraph {nodes [
   n1 = "Transform (Modify) Bone - Bone: dyn_hair_r_01"(Alpha=1, Translation="0,0,0", Rotation="8,0,10", Scale="1,1,1") {
    rotationMode=BMM_Additive
    rotationSpace=BCS_BoneSpace
    translationMode=BMM_Additive
    translationSpace=BCS_BoneSpace
    alphaBoolBlend {blendOption=Linear}
    boneToModify {boneName=dyn_hair_r_01}
    rotation {pitch=8, roll=10}
   }
   n2 = "Transform (Modify) Bone - Bone: dyn_hair_r_01"(ComponentPose=@n1, Alpha=1, Translation="0,0,0", Rotation="0,20,0", Scale="1,1,1") {
    rotationMode=BMM_Additive
    rotationSpace=BCS_BoneSpace
    translationMode=BMM_Additive
    translationSpace=BCS_BoneSpace
    alphaBoolBlend {blendOption=Linear}
    alphaScaleBiasClamp {bInterpResult=true}
    boneToModify {boneName=dyn_hair_r_01}
    rotation {yaw=20}
   }
   n3 = "Transform (Modify) Bone - Bone: dyn_hair_r_02"(ComponentPose=@n2, Alpha=1, Translation="0,0,0", Rotation="0,10,0", Scale="1,1,1") {
    rotationMode=BMM_Additive
    rotationSpace=BCS_BoneSpace
    translationMode=BMM_Additive
    translationSpace=BCS_BoneSpace
    alphaBoolBlend {blendOption=Linear}
    alphaScaleBiasClamp {bInterpResult=true}
    boneToModify {boneName=dyn_hair_r_02}
    rotation {yaw=10}
   }
   "Transform (Modify) Bone - Bone: dyn_hair_r_03"(ComponentPose=@n3, Alpha=1, Translation="0,0,0", Rotation="0,-24,0", Scale="1,1,1") {
    rotationMode=BMM_Additive
    rotationSpace=BCS_BoneSpace
    translationMode=BMM_Additive
    translationSpace=BCS_BoneSpace
    alphaBoolBlend {blendOption=Linear}
    alphaScaleBiasClamp {bInterpResult=true}
    boneToModify {boneName=dyn_hair_r_03}
    rotation {yaw=-24}
   }
   n4 = RigidBody!(
    ComponentPose=Local To Component(LocalPose=Use cached pose 'RB_BackPose'())
    bAlphaBoolEnabled:bool
    OverrideWorldGravity:Vector="0,0,-980"
    ComponentLinearAccScale:Vector="0,0,0"
    ComponentLinearVelScale:Vector="0,0,0"
   ) -> (Pose) {
    alphaInputType=Bool
    bClampLinearTranslationLimitToRefPose=true
    bOverrideWorldGravity=true
    overridePhysicsAsset=&/F_MED_CommandoSpy_FaceAcc_FrontHairPhysics_01
    simulationSpace=BaseBoneSpace
    alphaBoolBlend {blendOption=Linear}
    alphaScaleBiasClamp {interpSpeedDecreasing=4, interpSpeedIncreasing=6}
    baseBoneRef {boneName=spine_05}
    componentLinearAccScale=<0.4 0.4 0.4>
    componentLinearVelScale=<0.3 0.3 0.3>
    overrideWorldGravity {z=-875}
    simSpaceSettings {maxAngularAcceleration=200, maxAngularVelocity=5, maxLinearAcceleration=500, maxLinearVelocity=200}
   }
   n5 = Component To Local(ComponentPose=@n4)
   n6 = RigidBody(
    ComponentPose=Local To Component(LocalPose=Use cached pose 'RB_BackPose'())
    OverrideWorldGravity="0,0,-980"
    ComponentLinearAccScale="0,0,0"
    ComponentLinearVelScale="0,0,0"
   ) {
    alphaInputType=Bool
    bClampLinearTranslationLimitToRefPose=true
    bOverrideWorldGravity=true
    overridePhysicsAsset=&/F_MED_CommandoSpy_FaceAcc_FrontHairPhysics
    simulationSpace=BaseBoneSpace
    alphaBoolBlend {blendOption=Linear}
    alphaScaleBiasClamp {interpSpeedDecreasing=4, interpSpeedIncreasing=6}
    baseBoneRef {boneName=spine_05}
    componentLinearAccScale=<0.4 0.4 0.4>
    componentLinearVelScale=<0.3 0.3 0.3>
    overrideWorldGravity {z=-875}
    simSpaceSettings {maxAngularAcceleration=200, maxAngularVelocity=5, maxLinearAcceleration=500, maxLinearVelocity=200}
   }
   n7 = Component To Local(ComponentPose=@n6)
   n8 = Get bIsSkydiveDiveMode()
   n9 = Get IsWearingBackbling() -> (IsWearingBackbling:bool)
   n10 = "Constraint - dyn_hair_bk_l_03 (PPP)"!(
    ComponentPose=Local To Component(LocalPose=Blend Poses by bool(BlendPose_0=@n7, BlendPose_1=@n5, BlendTime_0=0.9, BlendTime_1=0.9, bActiveValue=@n8) {blendType=Linear})
    bAlphaBoolEnabled:bool=@n9
    ConstraintWeights_0=0.5
    ConstraintWeights_1:float=0.2
    ConstraintWeights_2:float=0.3
   ) -> (Pose) {
    alphaInputType=Bool
    alphaBoolBlend {blendOption=Linear}
    boneToModify {boneName=dyn_hair_bk_l_03}
    constraintSetup [{targetBone {boneName=spine_05}}, {targetBone {boneName=neck_01}}, {targetBone {boneName=clavicle_l}}]
    constraintWeights [1,1,1]
   }
   n11 = "Constraint - dyn_hair_bk_03 (PP)"!(ComponentPose=@n10, bAlphaBoolEnabled:bool=@n9, ConstraintWeights_0=0.5, ConstraintWeights_1:float=0.5) -> (Pose) {
    alphaInputType=Bool
    alphaBoolBlend {blendOption=Linear}
    boneToModify {boneName=dyn_hair_bk_03}
    constraintSetup [{targetBone {boneName=spine_05}}, {targetBone {boneName=neck_01}}]
    constraintWeights [1,1]
   }
   n12 = "Constraint - dyn_hair_bk_r_03 (PPP)"!(ComponentPose=@n11, bAlphaBoolEnabled:bool=@n9, ConstraintWeights_0=0.5, ConstraintWeights_1:float=0.2, ConstraintWeights_2:float=0.3) -> (Pose) {
    alphaInputType=Bool
    alphaBoolBlend {blendOption=Linear}
    boneToModify {boneName=dyn_hair_bk_r_03}
    constraintSetup [{targetBone {boneName=spine_05}}, {targetBone {boneName=neck_01}}, {targetBone {boneName=clavicle_r}}]
    constraintWeights [1,1,1]
   }
   n13 = "Constraint - dyn_hair_bk_l_04 (PPP)"!(ComponentPose=@n12, bAlphaBoolEnabled:bool=@n9, ConstraintWeights_0=0.5, ConstraintWeights_1:float=0.2, ConstraintWeights_2:float=0.3) -> (Pose) {
    alphaInputType=Bool
    alphaBoolBlend {blendOption=Linear}
    boneToModify {boneName=dyn_hair_bk_l_04}
    constraintSetup [{targetBone {boneName=spine_05}}, {targetBone {boneName=neck_01}}, {targetBone {boneName=clavicle_l}}]
    constraintWeights [1,1,1]
   }
   n14 = "Constraint - dyn_hair_bk_04 (PP)"!(ComponentPose=@n13, bAlphaBoolEnabled:bool=@n9, ConstraintWeights_0=0.5, ConstraintWeights_1:float=0.5) -> (Pose) {
    alphaInputType=Bool
    alphaBoolBlend {blendOption=Linear}
    boneToModify {boneName=dyn_hair_bk_04}
    constraintSetup [{targetBone {boneName=spine_05}}, {targetBone {boneName=neck_01}}]
    constraintWeights [1,1]
   }
   n15 = "Constraint - dyn_hair_bk_r_04 (PPP)"!(ComponentPose=@n14, bAlphaBoolEnabled:bool=@n9, ConstraintWeights_0=0.5, ConstraintWeights_1:float=0, ConstraintWeights_2:float=0.5) -> (Pose) {
    alphaInputType=Bool
    alphaBoolBlend {blendOption=Linear}
    boneToModify {boneName=dyn_hair_bk_r_04}
    constraintSetup [{targetBone {boneName=spine_05}}, {targetBone {boneName=neck_01}}, {targetBone {boneName=clavicle_r}}]
    constraintWeights [1,1,1]
   }
   n16 = "Constraint - dyn_hair_bk_l_03 (PPP)"(ComponentPose=@n15, bAlphaBoolEnabled=True, ConstraintWeights_0=0.5, ConstraintWeights_1=0.2, ConstraintWeights_2=0.3) {
    alphaInputType=Bool
    alphaBoolBlend {blendInTime=0.2, blendOption=Linear, blendOutTime=0.2}
    boneToModify {boneName=dyn_hair_bk_l_03}
    constraintSetup [{targetBone {boneName=spine_05}}, {targetBone {boneName=neck_01}}, {targetBone {boneName=clavicle_l}}]
    constraintWeights [1,1,1]
   }
   n17 = "Constraint - dyn_hair_bk_03 (PP)"(ComponentPose=@n16, bAlphaBoolEnabled=True, ConstraintWeights_0=0.5, ConstraintWeights_1=0.5) {
    alphaInputType=Bool
    alphaBoolBlend {blendInTime=0.2, blendOption=Linear, blendOutTime=0.2}
    boneToModify {boneName=dyn_hair_bk_03}
    constraintSetup [{targetBone {boneName=spine_05}}, {targetBone {boneName=neck_01}}]
    constraintWeights [1,1]
   }
   n18 = "Constraint - dyn_hair_bk_r_03 (PPP)"(ComponentPose=@n17, bAlphaBoolEnabled=True, ConstraintWeights_0=0.5, ConstraintWeights_1=0.2, ConstraintWeights_2=0.3) {
    alphaInputType=Bool
    alphaBoolBlend {blendInTime=0.2, blendOption=Linear, blendOutTime=0.2}
    boneToModify {boneName=dyn_hair_bk_r_03}
    constraintSetup [{targetBone {boneName=spine_05}}, {targetBone {boneName=neck_01}}, {targetBone {boneName=clavicle_r}}]
    constraintWeights [1,1,1]
   }
   n19 = "Constraint - dyn_hair_bk_l_04 (PPP)"(ComponentPose=@n18, bAlphaBoolEnabled=True, ConstraintWeights_0=0.5, ConstraintWeights_1=0.2, ConstraintWeights_2=0.3) {
    alphaInputType=Bool
    alphaBoolBlend {blendInTime=0.2, blendOption=Linear, blendOutTime=0.2}
    boneToModify {boneName=dyn_hair_bk_l_04}
    constraintSetup [{targetBone {boneName=spine_05}}, {targetBone {boneName=neck_01}}, {targetBone {boneName=clavicle_l}}]
    constraintWeights [1,1,1]
   }
   n20 = "Constraint - dyn_hair_bk_04 (PP)"(ComponentPose=@n19, bAlphaBoolEnabled=True, ConstraintWeights_0=0.5, ConstraintWeights_1=0.5) {
    alphaInputType=Bool
    alphaBoolBlend {blendInTime=0.2, blendOption=Linear, blendOutTime=0.2}
    boneToModify {boneName=dyn_hair_bk_04}
    constraintSetup [{targetBone {boneName=spine_05}}, {targetBone {boneName=neck_01}}]
    constraintWeights [1,1]
   }
   n21 = "Constraint - dyn_hair_bk_r_04 (PPP)"(ComponentPose=@n20, bAlphaBoolEnabled=True, ConstraintWeights_0=0.5, ConstraintWeights_1=0.2, ConstraintWeights_2=0.3) {
    alphaInputType=Bool
    alphaBoolBlend {blendInTime=0.2, blendOption=Linear, blendOutTime=0.2}
    boneToModify {boneName=dyn_hair_bk_r_04}
    constraintSetup [{targetBone {boneName=spine_05}}, {targetBone {boneName=neck_01}}, {targetBone {boneName=clavicle_r}}]
    constraintWeights [1,1,1]
   }
   n22 = Anim Dynamics - Bone: dyn_hair_bk_l_02!(ComponentPose=@n21, Alpha=1, GravityOverride:Vector="0,0,-980", ExternalForce="0,0,5") -> (Pose) {
    angularDampingOverride=0.9999
    angularSpringConstant=20000
    bAngularSpring=true
    bOverrideAngularDamping=true
    bUseGravityOverride=true
    gravityScale=0.35
    simulationSpace=RootRelative
    alphaBoolBlend {blendOption=Linear}
    alphaScaleBiasClamp {bInterpResult=true, interpSpeedIncreasing=2}
    boundBone {boneName=dyn_hair_bk_l_02}
    componentLinearAccScale=<1 1 1>
    componentLinearVelScale=<1 1 1>
    physicsBodyDefinitions [{
      boundBone {boneName=dyn_hair_bk_l_02}
      boxExtents=<25 30 30>
      constraintSetup {angularLimitsMax=<0 0 0>, angularLimitsMin=<0 0 -30>, angularTarget=<1 0 0>}
      localJointOffset=<2 0 0>
     }]
   }
   n23 = Anim Dynamics - Bone: dyn_hair_bk_02!(ComponentPose=@n22, Alpha=1, GravityOverride:Vector="0,0,-980", ExternalForce="0,0,5") -> (Pose) {
    angularDampingOverride=0.9999
    angularSpringConstant=20000
    bAngularSpring=true
    bOverrideAngularDamping=true
    bUseGravityOverride=true
    gravityScale=0.35
    simulationSpace=RootRelative
    alphaBoolBlend {blendOption=Linear}
    alphaScaleBiasClamp {bInterpResult=true, interpSpeedIncreasing=2}
    boundBone {boneName=dyn_hair_bk_02}
    componentLinearAccScale=<1 1 1>
    componentLinearVelScale=<1 1 1>
    physicsBodyDefinitions [{
      boundBone {boneName=dyn_hair_bk_02}
      boxExtents=<25 30 30>
      constraintSetup {angularLimitsMax=<0 0 0>, angularLimitsMin=<0 0 -30>, angularTarget=<1 0 0>}
      localJointOffset=<2 0 0>
     }]
   }
   n24 = Anim Dynamics - Bone: dyn_hair_bk_r_02!(ComponentPose=@n23, Alpha=1, GravityOverride:Vector="0,0,-980", ExternalForce="0,0,5") -> (Pose) {
    angularDampingOverride=0.9999
    angularSpringConstant=20000
    bAngularSpring=true
    bOverrideAngularDamping=true
    bUseGravityOverride=true
    gravityScale=0.35
    simulationSpace=RootRelative
    alphaBoolBlend {blendOption=Linear}
    alphaScaleBiasClamp {bInterpResult=true, interpSpeedIncreasing=2}
    boundBone {boneName=dyn_hair_bk_r_02}
    componentLinearAccScale=<1 1 1>
    componentLinearVelScale=<1 1 1>
    physicsBodyDefinitions [{
      boundBone {boneName=dyn_hair_bk_r_02}
      boxExtents=<25 30 30>
      constraintSetup {angularLimitsMax=<0 0 0>, angularLimitsMin=<0 0 -30>, angularTarget=<1 0 0>}
      localJointOffset=<2 0 0>
     }]
   }
   n25 = Anim Dynamics - Bone: dyn_hair_bk_l_03!(ComponentPose=@n24, Alpha=1, GravityOverride:Vector="0,0,-980", ExternalForce="0,0,5") -> (Pose) {
    angularDampingOverride=0.9999
    angularSpringConstant=20000
    bAngularSpring=true
    bOverrideAngularDamping=true
    bUseGravityOverride=true
    gravityScale=0.35
    simulationSpace=RootRelative
    alphaBoolBlend {blendOption=Linear}
    alphaScaleBiasClamp {bInterpResult=true, interpSpeedIncreasing=2}
    boundBone {boneName=dyn_hair_bk_l_03}
    componentLinearAccScale=<1 1 1>
    componentLinearVelScale=<1 1 1>
    physicsBodyDefinitions [{
      boundBone {boneName=dyn_hair_bk_l_03}
      boxExtents=<25 30 30>
      constraintSetup {angularLimitsMax=<0 0 0>, angularLimitsMin=<0 0 -30>, angularTarget=<1 0 0>}
      localJointOffset=<2 0 0>
     }]
   }
   n26 = Anim Dynamics - Bone: dyn_hair_bk_03!(ComponentPose=@n25, Alpha=1, GravityOverride:Vector="0,0,-980", ExternalForce="0,0,5") -> (Pose) {
    angularDampingOverride=0.9999
    angularSpringConstant=20000
    bAngularSpring=true
    bOverrideAngularDamping=true
    bUseGravityOverride=true
    gravityScale=0.35
    simulationSpace=RootRelative
    alphaBoolBlend {blendOption=Linear}
    alphaScaleBiasClamp {bInterpResult=true, interpSpeedIncreasing=2}
    boundBone {boneName=dyn_hair_bk_03}
    componentLinearAccScale=<1 1 1>
    componentLinearVelScale=<1 1 1>
    physicsBodyDefinitions [{
      boundBone {boneName=dyn_hair_bk_03}
      boxExtents=<25 30 30>
      constraintSetup {angularLimitsMax=<0 0 0>, angularLimitsMin=<0 0 -30>, angularTarget=<1 0 0>}
      localJointOffset=<2 0 0>
     }]
   }
   n27 = Anim Dynamics - Bone: dyn_hair_bk_r_03!(ComponentPose=@n26, Alpha=1, GravityOverride:Vector="0,0,-980", ExternalForce="0,0,5") -> (Pose) {
    angularDampingOverride=0.9999
    angularSpringConstant=20000
    bAngularSpring=true
    bOverrideAngularDamping=true
    bUseGravityOverride=true
    gravityScale=0.35
    simulationSpace=RootRelative
    alphaBoolBlend {blendOption=Linear}
    alphaScaleBiasClamp {bInterpResult=true, interpSpeedIncreasing=2}
    boundBone {boneName=dyn_hair_bk_r_03}
    componentLinearAccScale=<1 1 1>
    componentLinearVelScale=<1 1 1>
    physicsBodyDefinitions [{
      boundBone {boneName=dyn_hair_bk_r_03}
      boxExtents=<25 30 30>
      constraintSetup {angularLimitsMax=<0 0 0>, angularLimitsMin=<0 0 -30>, angularTarget=<1 0 0>}
      localJointOffset=<2 0 0>
     }]
   }
   n28 = Anim Dynamics - Bone: dyn_hair_bk_l_04!(ComponentPose=@n27, Alpha=1, GravityOverride:Vector="0,0,-980", ExternalForce="0,0,5") -> (Pose) {
    angularDampingOverride=0.9999
    angularSpringConstant=20000
    bAngularSpring=true
    bOverrideAngularDamping=true
    bUseGravityOverride=true
    gravityScale=0.35
    simulationSpace=RootRelative
    alphaBoolBlend {blendOption=Linear}
    alphaScaleBiasClamp {bInterpResult=true, interpSpeedIncreasing=2}
    boundBone {boneName=dyn_hair_bk_l_04}
    componentLinearAccScale=<1 1 1>
    componentLinearVelScale=<1 1 1>
    physicsBodyDefinitions [{
      boundBone {boneName=dyn_hair_bk_l_04}
      boxExtents=<25 30 30>
      constraintSetup {angularLimitsMax=<0 0 0>, angularLimitsMin=<0 0 -30>, angularTarget=<1 0 0>}
      localJointOffset=<2 0 0>
     }]
   }
   n29 = Anim Dynamics - Bone: dyn_hair_bk_04!(ComponentPose=@n28, Alpha=1, GravityOverride:Vector="0,0,-980", ExternalForce="0,0,5") -> (Pose) {
    angularDampingOverride=0.9999
    angularSpringConstant=20000
    bAngularSpring=true
    bOverrideAngularDamping=true
    bUseGravityOverride=true
    gravityScale=0.35
    simulationSpace=RootRelative
    alphaBoolBlend {blendOption=Linear}
    alphaScaleBiasClamp {bInterpResult=true, interpSpeedIncreasing=2}
    boundBone {boneName=dyn_hair_bk_04}
    componentLinearAccScale=<1 1 1>
    componentLinearVelScale=<1 1 1>
    physicsBodyDefinitions [{
      boundBone {boneName=dyn_hair_bk_04}
      boxExtents=<25 30 30>
      constraintSetup {angularLimitsMax=<0 0 0>, angularLimitsMin=<0 0 -30>, angularTarget=<1 0 0>}
      localJointOffset=<2 0 0>
     }]
   }
   n30 = Anim Dynamics - Bone: dyn_hair_bk_r_04!(ComponentPose=@n29, Alpha=1, GravityOverride:Vector="0,0,-980", ExternalForce="0,0,5") -> (Pose) {
    angularDampingOverride=0.9999
    angularSpringConstant=20000
    bAngularSpring=true
    bOverrideAngularDamping=true
    bUseGravityOverride=true
    gravityScale=0.35
    simulationSpace=RootRelative
    alphaBoolBlend {blendOption=Linear}
    alphaScaleBiasClamp {bInterpResult=true, interpSpeedIncreasing=2}
    boundBone {boneName=dyn_hair_bk_r_04}
    componentLinearAccScale=<1 1 1>
    componentLinearVelScale=<1 1 1>
    physicsBodyDefinitions [{
      boundBone {boneName=dyn_hair_bk_r_04}
      boxExtents=<25 30 30>
      constraintSetup {angularLimitsMax=<0 0 0>, angularLimitsMin=<0 0 -30>, angularTarget=<1 0 0>}
      localJointOffset=<2 0 0>
     }]
   }
   n31 = "Transform (Modify) Bone - Bone: dyn_hair_bk_r_03"(ComponentPose=@n30, Alpha=1, Translation="0,1,-6", Rotation="20,5,0.022", Scale="1,1,1") {
    rotationMode=BMM_Additive
    rotationSpace=BCS_BoneSpace
    translationMode=BMM_Additive
    translationSpace=BCS_BoneSpace
    alphaBoolBlend {blendOption=Linear}
    boneToModify {boneName=dyn_hair_bk_r_03}
    rotation=<rot 20 5 0.022>
    translation {y=1, z=-6}
   }
   n32 = "Transform (Modify) Bone - Bone: dyn_hair_bk_r_04"(ComponentPose=@n31, Alpha=1, Translation="0.323,-3,0.042", Rotation="-0.6032,7,-6", Scale="1,1,1") {
    rotationMode=BMM_Additive
    rotationSpace=BCS_BoneSpace
    translationMode=BMM_Additive
    translationSpace=BCS_BoneSpace
    alphaBoolBlend {blendOption=Linear}
    boneToModify {boneName=dyn_hair_bk_r_04}
    rotation=<rot -0.6032 7 -6>
    translation=<0.323 -3 0.042>
   }
   n33 = "Transform (Modify) Bone - Bone: dyn_hair_r_01"(ComponentPose=@n32, Alpha=1, Translation="0,0,0", Rotation="0,0,0", Scale="1,1,1") {
    rotationMode=BMM_Additive
    rotationSpace=BCS_BoneSpace
    translationMode=BMM_Additive
    translationSpace=BCS_BoneSpace
    alphaBoolBlend {blendOption=Linear}
    boneToModify {boneName=dyn_hair_r_01}
   }
   n34 = "Transform (Modify) Bone - Bone: dyn_hair_r_02"(ComponentPose=@n33, Alpha=1, Translation="0,0,0", Rotation="0,0,0", Scale="1,1,1") {
    rotationMode=BMM_Additive
    rotationSpace=BCS_BoneSpace
    translationMode=BMM_Additive
    translationSpace=BCS_BoneSpace
    alphaBoolBlend {blendOption=Linear}
    boneToModify {boneName=dyn_hair_r_02}
   }
   n35 = "Transform (Modify) Bone - Bone: dyn_hair_r_03"(ComponentPose=@n34, Alpha=1, Translation="0,0,0", Rotation="0,0,0", Scale="1,1,1") {
    rotationMode=BMM_Additive
    rotationSpace=BCS_BoneSpace
    translationMode=BMM_Additive
    translationSpace=BCS_BoneSpace
    alphaBoolBlend {blendOption=Linear}
    boneToModify {boneName=dyn_hair_r_03}
   }
   n36 = "Transform (Modify) Bone - Bone: dyn_hair_r_01"(ComponentPose=@n35, Alpha=1, Translation="-1.115,0.000,-0.501", Rotation="33.989,-22.785,-7.676", Scale="1,1,1") {
    rotationMode=BMM_Additive
    rotationSpace=BCS_BoneSpace
    translationMode=BMM_Additive
    translationSpace=BCS_BoneSpace
    alphaBoolBlend {blendOption=Linear}
    boneToModify {boneName=dyn_hair_r_01}
    rotation=<rot 33.989 -22.785 -7.676>
    translation {x=-1.115, z=-0.501}
   }
   n37 = "Transform (Modify) Bone - Bone: dyn_hair_r_02"(ComponentPose=@n36, Alpha=1, Translation="-0.919,0.449,0.455", Rotation="-30.050,59.576,-16.409", Scale="1,1,1") {
    rotationMode=BMM_Additive
    rotationSpace=BCS_BoneSpace
    translationMode=BMM_Additive
    translationSpace=BCS_BoneSpace
    alphaBoolBlend {blendOption=Linear}
    boneToModify {boneName=dyn_hair_r_02}
    rotation=<rot -30.05 59.576 -16.409>
    translation=<-0.919 0.449 0.455>
   }
   n38 = "Transform (Modify) Bone - Bone: dyn_hair_r_03"(ComponentPose=@n37, Alpha=1, Translation="0,0,0", Rotation="-1.239,-54.323,-7.750", Scale="1,1,1") {
    rotationMode=BMM_Additive
    rotationSpace=BCS_BoneSpace
    translationMode=BMM_Additive
    translationSpace=BCS_BoneSpace
    alphaBoolBlend {blendOption=Linear}
    boneToModify {boneName=dyn_hair_r_03}
    rotation=<rot -1.239 -54.323 -7.75>
   }
   n39 = "Transform (Modify) Bone - Bone: dyn_hair_bk_l_02"!(ComponentPose=@n38, bAlphaBoolEnabled:bool=True, Translation="0,0,0", Rotation="0,10,0", Scale="1,1,1") -> (Pose) {
    alphaInputType=Bool
    rotationMode=BMM_Additive
    rotationSpace=BCS_BoneSpace
    translationMode=BMM_Additive
    translationSpace=BCS_BoneSpace
    alphaBoolBlend {blendOption=Linear}
    boneToModify {boneName=dyn_hair_bk_l_02}
    rotation {yaw=10}
   }
   n40 = Control Rig!(Source=Component To Local(ComponentPose=@n39), Alpha, TakeOut:bool=Get Takeout_Emo() -> (Takeout_Emo:bool)) -> (Pose) {
    alphaBoolBlend {blendOption=Linear}
    controlRigAssetReference {blueprintRigClass=&/F_MED_CommandoSpy_FaceAcc_Post_CtrlRig.F_MED_CommandoSpy_FaceAcc_Post_CtrlRig_C}
   }
   Output Pose(Result=@n40) {layerGroup=None}
   n41 = Control Rig!(
    Source=Copy Pose From Mesh() {bUseAttachedParent=true}
    Alpha
    FE Idle:bool=Get FrontEndIdle() -> (FrontEndIdle:bool)
    targeting:bool=Get bIsTargeting()
   ) -> (Pose) {
    alphaBoolBlend {blendOption=Linear}
    controlRigAssetReference {blueprintRigClass=&/F_MED_CommandoSpy_FaceAcc_CtrlRig.F_MED_CommandoSpy_FaceAcc_CtrlRig_C}
   }
   n42 = Copy Pose From Mesh() {bUseAttachedParent=true}
   n43 = Get bIsSkyDiving()
   n44 = "Transform (Modify) Bone - Bone: upperarm_l"!(
    ComponentPose=Local To Component(LocalPose=Blend Poses by bool(BlendPose_0=@n42, BlendPose_1=@n41, bActiveValue=@n43) {blendType=Linear})
    Alpha=1
    Rotation="0,0,0"
   ) -> (Pose) {
    rotationMode=BMM_Replace
    rotationSpace=BCS_WorldSpace
    scaleMode=BMM_Additive
    translationSpace=BCS_WorldSpace
    alphaBoolBlend {blendInTime=1, blendOption=Linear, blendOutTime=1}
    boneToModify {boneName=upperarm_l}
   }
   n45 = "Transform (Modify) Bone - Bone: upperarm_r"!(ComponentPose=@n44, Alpha=1, Rotation="0,0,0") -> (Pose) {
    rotationMode=BMM_Replace
    rotationSpace=BCS_WorldSpace
    scaleMode=BMM_Additive
    translationSpace=BCS_WorldSpace
    alphaBoolBlend {blendInTime=1, blendOption=Linear, blendOutTime=1}
    boneToModify {boneName=upperarm_r}
   }
   Save cached pose 'StartPose'(Pose=Component To Local(ComponentPose=@n45)) {cacheName=StartPose}
   n46 = RigidBody!(
    ComponentPose=Local To Component(LocalPose=Use cached pose 'StartPose'())
    Alpha:float
    OverrideWorldGravity="0,0,-980"
    ComponentLinearAccScale="0,0,0"
    ComponentLinearVelScale="0,0,0"
   ) -> (Pose) {
    bClampLinearTranslationLimitToRefPose=true
    bOverrideWorldGravity=true
    overridePhysicsAsset=&/F_MED_CommandoSpy_FaceAcc_BackHairPhysics
    simulationSpace=BaseBoneSpace
    alphaBoolBlend {blendOption=Linear}
    alphaScaleBiasClamp {interpSpeedDecreasing=4, interpSpeedIncreasing=6}
    baseBoneRef {boneName=spine_05}
    componentLinearAccScale=<0.4 0.4 0.4>
    componentLinearVelScale=<0.3 0.3 0.3>
    overrideWorldGravity {z=-875}
    simSpaceSettings {maxAngularAcceleration=200, maxAngularVelocity=5, maxLinearAcceleration=500, maxLinearVelocity=200}
   }
   n47 = Component To Local(ComponentPose=@n46)
   n48 = "Transform (Modify) Bone - Bone: dyn_hair_r_MSTR"!(ComponentPose=Local To Component(LocalPose=Use cached pose 'StartPose'()), Alpha=1, Translation="0,0,0", Rotation="0,0,0") -> (Pose) {
    rotationMode=BMM_Additive
    rotationSpace=BCS_BoneSpace
    scaleMode=BMM_Additive
    translationMode=BMM_Additive
    translationSpace=BCS_BoneSpace
    alphaBoolBlend {blendInTime=1, blendOption=Linear, blendOutTime=1}
    boneToModify {boneName=dyn_hair_r_MSTR}
   }
   n49 = RigidBody(ComponentPose=@n48, OverrideWorldGravity="0,0,-980", ComponentLinearAccScale="0,0,0", ComponentLinearVelScale="0,0,0") {
    bClampLinearTranslationLimitToRefPose=true
    bOverrideWorldGravity=true
    overridePhysicsAsset=&F_MED_CommandoSpy_FaceAcc_BackHairPhysics
    simulationSpace=BaseBoneSpace
    alphaBoolBlend {blendOption=Linear}
    alphaScaleBiasClamp {interpSpeedDecreasing=4, interpSpeedIncreasing=6}
    baseBoneRef {boneName=spine_05}
    componentLinearAccScale=<0.4 0.4 0.4>
    componentLinearVelScale=<0.3 0.3 0.3>
    overrideWorldGravity {z=-875}
    simSpaceSettings {maxAngularAcceleration=200, maxAngularVelocity=5, maxLinearAcceleration=500, maxLinearVelocity=200}
   }
   n50 = Component To Local(ComponentPose=@n49)
   n51 = Get bIsSkyDiving()
   n52 = "Transform (Modify) Bone - Bone: dyn_hair_bk_l_03"(
    ComponentPose=Local To Component(LocalPose=Blend Poses by bool(BlendPose_0=@n50, BlendPose_1=@n47, BlendTime_0=0.3, BlendTime_1=0.3, bActiveValue=@n51) {blendType=Linear})
    Alpha=1
    Translation="0,0,0"
    Rotation="0,-40,0"
    Scale="1,1,1"
   ) {
    rotationMode=BMM_Additive
    rotationSpace=BCS_BoneSpace
    translationMode=BMM_Additive
    translationSpace=BCS_BoneSpace
    alphaBoolBlend {blendOption=Linear}
    alphaScaleBiasClamp {bInterpResult=true, interpSpeedIncreasing=2}
    boneToModify {boneName=dyn_hair_bk_l_03}
    rotation {yaw=-40}
   }
   n53 = "Transform (Modify) Bone - Bone: dyn_hair_bk_03"(ComponentPose=@n52, Alpha=1, Translation="0,0,0", Rotation="0,-40,0", Scale="1,1,1") {
    rotationMode=BMM_Additive
    rotationSpace=BCS_BoneSpace
    translationMode=BMM_Additive
    translationSpace=BCS_BoneSpace
    alphaBoolBlend {blendOption=Linear}
    alphaScaleBiasClamp {bInterpResult=true, interpSpeedIncreasing=2}
    boneToModify {boneName=dyn_hair_bk_03}
    rotation {yaw=-40}
   }
   n54 = "Transform (Modify) Bone - Bone: dyn_hair_bk_r_03"(ComponentPose=@n53, Alpha=1, Translation="0,0,0", Rotation="0,-40,0", Scale="1,1,1") {
    rotationMode=BMM_Additive
    rotationSpace=BCS_BoneSpace
    translationMode=BMM_Additive
    translationSpace=BCS_BoneSpace
    alphaBoolBlend {blendOption=Linear}
    alphaScaleBiasClamp {bInterpResult=true, interpSpeedIncreasing=2}
    boneToModify {boneName=dyn_hair_bk_r_03}
    rotation {yaw=-40}
   }
   Save cached pose 'RB_BackPose'(Pose=Component To Local(ComponentPose=@n54)) {cacheName=RB_BackPose}
  ]}
 EventGraph {nodes [
   Event BlueprintUpdateAnimation()
   > n1 = Sequence!(execute) -> (
    then_0
    then_1
    then_2:Exec
    then_3:Exec
    then_4:Exec
    then_5:Exec
    then_6:Exec
    then_7:Exec
    then_8:Exec
    then_9:Exec
    then_10:Exec
    then_11:Exec
    then_12:Exec
    then_13:Exec
   )
   n2 = "Gravity Override State Function (Parameter Asset) (Nativized)"(
    execute=@n1.then_1
    CharacterPartAnimInstance=Self-Reference()
    GameTimeInSeconds=GetGameTimeInSeconds()
    Parameters=&/F_MED_CommandoSpy_Parameters
   )
   > Set Gravity Override(Gravity Override=@n2)
   n3 = Get BodyMesh() -> (BodyMesh:Skeletal Mesh Component Object Reference)
   > Is Valid(InputObject=@n3)
   > Set Upperarm_L_Rotation(Upperarm_L_Rotation:Rotator=BreakTransform(InTransform=GetSocketTransform(self=@n3, InSocketName=upperarm_l, TransformSpace=RTS_World)).Rotation) -> (Output_Get:Rotator)
   > Set Upperarm_R_Rotation(Upperarm_R_Rotation:Rotator=BreakTransform(InTransform=GetSocketTransform(self=@n3, InSocketName=upperarm_r, TransformSpace=RTS_World)).Rotation) -> (Output_Get:Rotator)
   n4 = Get bIsPlayingEmote()
   Set FrontEndIdle(execute=@n1.then_10, FrontEndIdle:bool=OR Boolean(A=Get bIsInFrontEnd(), B=NOT Boolean(A=@n4))) -> (Output_Get:bool)
   n5 = Get bIsInFrontEnd()
   Set Not FE(execute=@n1.then_12, Not FE:bool=OR Boolean(A=NOT Boolean(A=@n5), B=false)) -> (Output_Get:bool)
   n6 = Get bIsCrouching()
   n7 = Get bIsMoving2D()
   n8 = AND Boolean(A=NOT Boolean(A=Get bIsTargeting()), B=NOT Boolean(A=AND Boolean(A=@n6, B=@n7)))
   n9 = float * float(
    A=GetAlphaFromReferenceTransformDelta(SkeletalMeshComponent=Get Skel Mesh(), BoneName=spine_05, BaseName=head, DeltaComponentType=EulerX, InputRange=(X=0,Y=90), TwistAxis="1,0,0")
    B=1
   ) {is=title K2Node_PromotableOperator}
   Set Head_Right_Twist(execute=@n1.then_13, Head_Right_Twist:double="Clamp (Float)"(Value=float * float(A=@n9, B:bool=@n8), Min=0, Max=1)) -> (Output_Get:double)
   n10 = Get bIsCrouching()
   n11 = Get bIsMoving2D()
   n12 = AND Boolean(A=NOT Boolean(A=Get bIsTargeting()), B=NOT Boolean(A=AND Boolean(A=@n10, B=@n11)))
   n13 = float * float(
    A=GetAlphaFromReferenceTransformDelta(
     SkeletalMeshComponent=Get Skel Mesh()
     BoneName=spine_05
     BaseName=head
     DeltaComponentType=EulerX
     InputRange=(X=0,Y=-90)
     TwistAxis="1,0,0"
    )
    B:double=1
   )
   > Set Head_Left_Twist(Head_Left_Twist:double="Clamp (Float)"(Value=float * float(A=@n13, B:bool=@n12), Min=0, Max=1)) -> (Output_Get:double)
   n14 = GetOwnerSkeletalMeshComponent()
   n15 = BreakRotator(InRot=BreakTransform(InTransform=GetDeltaTransformFromRefPose(self=@n14, BoneName=thigh_r, BaseName=pelvis)).Rotation)
   n16 = Get bIsFalling()
   n17 = Get bIsJumping()
   Set jump(execute=@n1.then_2, jump:double=float * float(A=MapRangeClamped(Value=@n15.Yaw, InRangeA=0, InRangeB=-45), B=OR Boolean(A=@n17, B=@n16))) -> (Output_Get:double)
   n18 = NOT Boolean(A=Get IsWearingBackbling())
   n19 = Get bIsSkydiveFloating()
   Set Skydive Backling(execute=@n1.then_3, Skydive Backling:double="To Float (Boolean)"(InBool=AND Boolean(A=@n18, B=@n19))) -> (Output_Get:double)
   n20 = Cast To FortPlayerAnimInstance(execute=@n1.then_5, Object=GetAnimInstance(self=GetOwnerSkeletalMeshComponent()))
   > n21 = Switch on EFortWeaponCoreAnimation!(execute, Selection:EFortWeaponCoreAnimation Enum=Get WeaponCoreAnim(self=@n20)) -> (
    Melee:Exec
    Pistol:Exec
    Shotgun:Exec
    PaperBlueprint:Exec
    Rifle:Exec
    MeleeOneHand:Exec
    MachinePistol:Exec
    RocketLauncher:Exec
    GrenadeLauncher:Exec
    GoingCommando:Exec
    AssaultRifle:Exec
    TacticalShotgun:Exec
    SniperRifle:Exec
    TrapPlacement:Exec
    ShoulderLauncher:Exec
    AbilityDecoTool:Exec
    Crossbow:Exec
    C4:Exec
    RemoteControl:Exec
    DualWield:Exec
    AR_BullPup:Exec
    AR_ForwardGrip:Exec
    MedPackPaddles:Exec
    SMG_P90:Exec
    AR_DrumGun:Exec
    Consumable_Small:Exec
    Consumable_Large:Exec
    Balloon:Exec
    MountedTurret:Exec
    CreativeTool:Exec
    ExplosiveBow:Exec
    AshtonIndigo:Exec
    AshtonChicago:Exec
    MeleeDualWield:Exec
    Unarmed:Exec
   )
   n22 = Branch(
    execute=[
     @n21.AR_BullPup
     @n21.AR_DrumGun
     @n21.AR_ForwardGrip
     @n21.AssaultRifle
     @n21.Rifle
     @n21.SMG_P90
     @n21.Shotgun
     @n21.SniperRifle
     @n21.TacticalShotgun
     @n21.TrapPlacement
    ]
    Condition=Get bIsTargeting()
   )
   > Set Riffle Equipped(Riffle Equipped:bool=true) -> (Output_Get:bool)
   Set Riffle Equipped(execute=@n22.else, Riffle Equipped=false)
   n23 = Branch(execute=@n21.Pistol, Condition=Get bIsTargeting())
   > Set Pistol Equipped(Pistol Equipped:bool=true) -> (Output_Get:bool)
   Set Pistol Equipped(execute=@n23.else, Pistol Equipped=false)
   n24 = "To Float (Boolean)"(InBool=Get Riffle Equipped() -> (Riffle Equipped:bool))
   n25 = "To Float (Boolean)"(InBool=NOT Boolean(A=Get IsWearingBackbling()))
   n26 = float * float(
    A=GetAlphaFromReferenceTransformDelta(SkeletalMeshComponent=Get Skel Mesh(), BoneName=spine_05, BaseName=head, InputRange=(X=0,Y=90), TwistAxis="0,0,1")
    B:double=3.3
   )
   Set RiffleAimingUp(
    execute=@n1.then_6
    RiffleAimingUp:double="Clamp (Float)"(Value=float * float!(A=@n26, B=@n25, C:double=@n24) -> (ReturnValue), Min=0, Max=1)
   ) -> (Output_Get:double)
   n27 = "To Float (Boolean)"(InBool=NOT Boolean(A=Get IsWearingBackbling()))
   n28 = "To Float (Boolean)"(InBool=Get Riffle Equipped())
   n29 = float * float!(
    A=GetAlphaFromReferenceTransformDelta(SkeletalMeshComponent=Get Skel Mesh(), BoneName=spine_05, BaseName=head, InputRange=(X=0,Y=-90), TwistAxis="0,0,1")
    B=3.3
   ) -> (ReturnValue)
   > Set RiffleAimingDown(RiffleAimingDown:double="Clamp (Float)"(Value=float * float!(A=@n29, B=@n27, C:double=@n28) -> (ReturnValue), Min=0, Max=1)) -> (Output_Get:double)
   n30 = GetAnimInstance(self=GetOwnerSkeletalMeshComponent())
   Is Valid(exec=@n1.then_7, InputObject=@n30)
   > Set EasternBlocEmote(EasternBlocEmote:bool=OR Boolean(A=Montage_IsPlaying(self=@n30), B=false)) -> (Output_Get:bool)
   > Set Emote_Celebration(Emote_Celebration:bool=OR Boolean(
     A=Montage_IsPlaying(self=@n30, Montage=/BRCosmetics/Animation/Game/MainPlayer/Emotes/Emote_Celebration_Female_M)
     B=Montage_IsPlaying(self=@n30, Montage=&Emote_Celebration_Female_M)
    )) -> (Output_Get:bool)
   n31 = GetAnimInstance(self=GetOwnerSkeletalMeshComponent())
   Is Valid(exec=@n1.then_8, InputObject=@n31)
   > Set EasternBlocEmote(EasternBlocEmote=OR Boolean(A=Montage_IsPlaying(self=@n31), B=false))
   > Set Takeout_Emo(Takeout_Emo:bool=OR Boolean(
     A=Montage_IsPlaying(self=@n31, Montage=/BRCosmetics/Animation/Game/MainPlayer/Emotes/Takeout/Emote_Takeout_CLF_M)
     B=Montage_IsPlaying(self=@n31, Montage=&Emote_Takeout_CLF_M)
    )) -> (Output_Get:bool)
   n32 = Get Skel Mesh()
   Set Head Up(
    execute=@n1.then_11
    Head Up:double=GetAlphaFromReferenceTransformDelta(SkeletalMeshComponent=@n32, BoneName=spine_05, BaseName=head, InputRange=(X=10,Y=70), TwistAxis="0,0,1")
   ) -> (Output_Get:double)
   > Set Head Down(Head Down:double=GetAlphaFromReferenceTransformDelta(SkeletalMeshComponent=@n32, BoneName=spine_05, BaseName="head ", InputRange=(X=0,Y=-90), OutputRange=(X=0,Y=-1), TwistAxis="0,0,1")) -> (Output_Get:double)
   > Set Head_Turn_L(Head_Turn_L:double=GetAlphaFromReferenceTransformDelta(SkeletalMeshComponent=@n32, BoneName=spine_05, BaseName=head, DeltaComponentType=EulerX, InputRange=(X=0,Y=-22), TwistAxis="1,0,0")) -> (Output_Get:double)
   > Set Head_Turn_R(Head_Turn_R:double=GetAlphaFromReferenceTransformDelta(SkeletalMeshComponent=@n32, BoneName=spine_05, BaseName=head, DeltaComponentType=EulerX, InputRange=(X=0,Y=22), TwistAxis="1,0,0")) -> (Output_Get:double)
   n33 = "To Float (Boolean)"(InBool=Get Pistol Equipped() -> (Pistol Equipped:bool))
   n34 = float * float!(
    A=GetAlphaFromReferenceTransformDelta(SkeletalMeshComponent=Get Skel Mesh(), BoneName=spine_05, BaseName=head, InputRange=(X=0,Y=90), TwistAxis="0,0,1")
    B=3.3
   ) -> (ReturnValue)
   Set AimingUp(execute=@n1.then_4, AimingUp:double="Clamp (Float)"(Value=float * float(A=@n34, B=@n33), Min=0, Max=1)) -> (Output_Get:double)
   n35 = GetAnimInstance(self=GetOwnerSkeletalMeshComponent())
   Is Valid(exec=@n1.then_9, InputObject=@n35)
   > Set EasternBlocEmote(EasternBlocEmote=OR Boolean(A=Montage_IsPlaying(self=@n35), B=false))
   > Set Griddles_Emo(Griddles_Emo:bool=OR Boolean(
     A=Montage_IsPlaying(self=@n35, Montage=/BRCosmetics/Animation/Game/MainPlayer/Emotes/Griddles/Emote_Griddle_CMF_M)
     B=Montage_IsPlaying(self=@n35, Montage=&Emote_Griddle_CMF_M)
    )) -> (Output_Get:bool)
   n36 = CheckForBackbling()
   Set IsWearingBackbling(execute=@n1.then_0, IsWearingBackbling:bool=@n36) -> (Output_Get:bool)
   > Set Is Backbling A Cape(Is Backbling A Cape:bool=@n36.bIsBackblingACape) -> (Output_Get:bool)
   > Set is Wear a Cape(is Wear a Cape:bool=AND Boolean(A=@n36, B=@n36.bIsBackblingLong)) -> (Output_Get:bool)
   n37 = Event OnCharacterCustomizationLoaded()
   > Set BodyMesh(BodyMesh=Get Array Item with EFort Custom Part Type(Array=@n37.CustomMeshParts))
   Event BlueprintInitializeAnimation()
   > Set Skel Mesh(Skel Mesh=GetOwnerSkeletalMeshComponent())
   PrintString!(execute, InString=Hello, bPrintToScreen, bPrintToLog, TextColor, Duration, Key) -> (then)
   Set WearingBack(WearingBack:bool=CheckForBackbling()) -> (Output_Get:bool)
  ]}
]
variables [
 AimingUp {type=number}
 BodyMesh {type=SkeletalMeshComponent}
 EasternBlocEmote {type=boolean}
 Emote_Celebration {type=boolean}
 FE Idle {type=boolean}
 FrontEndIdle {type=boolean}
 Gravity Override {type=Vector}
 Griddles_Emo {type=boolean}
 HandOnHips FE {type=boolean}
 Head Down {type=number}
 Head Up {type=number}
 Head_Left_Twist {type=number}
 Head_Right_Twist {type=number}
 Head_Turn_L {type=number}
 Head_Turn_R {type=number}
 Is Backbling A Cape {type=boolean}
 is Wear a Cape {type=boolean}
 IsWearingBackbling {type=boolean}
 jump {type=number}
 Not FE {type=boolean}
 Pistol Equipped {type=boolean}
 Riffle Equipped {type=boolean}
 RiffleAimingDown {type=number}
 RiffleAimingUp {type=number}
 Skel Mesh {type=SkeletalMeshComponent}
 Skydive Backling {type=number}
 Takeout_Emo {type=boolean}
 Upperarm_L_Rotation {type=Rotator}
 Upperarm_R_Rotation {type=Rotator}
 WearingBack {type=boolean}
]
